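# Niagara idle effect for the HighTower Squash one-handed pickaxe, combining flickering lightning, flares, spinning debris and glow sprites.
# Short randomized loops fire bursts of mesh blades, bone-flow bolts and hand arcs, oriented from sockets or bones sampled on User.InstigatorMesh.
# The arc emitter also drives offset meshes and an HLSL-gated light on one particle per burst.
# System.RandomColor tints some layers; one decal emitter is disabled.
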NiagaraSystem /BRCosmetics/Weapons/FORT_Melee/Pickaxe_HighTower_Squash_1H/FX/NS_PIckaxe_HighTowerSquash_1H_Idle
dependencies {
 usedDataInterfaces [&NiagaraDataInterfaceColorCurve, &NiagaraDataInterfaceCurve]
 usedDynamicInputs []
 usedModules [
  &SystemState
  &self:SystemSpawnScript.SystemScriptSource.SystemScriptGraph.NiagaraNodeAssignment_0.SetVariables_#1
 ]
}
emitters {
 Debris_Vortex {
  emitterUpdateScript [
   &EmitterState(
    Life Cycle Mode=::NewEnumerator1
    Loop Delay=0.5
    Loop Duration=&UniformRangedFloat(Minimum=0.75, Maximum=1)
    Recalculate Duration Each Loop=true
    UseLoopDelay=false
   )
   &SpawnRate(SpawnRate=&UniformRangedFloat(Minimum=7, Maximum=10), Use Spawn Probability=false)
  ]
  particleSpawnScript [
   /Niagara/Modules/Spawn/Initialization/InitializeParticle(
    Color=&UniformRangedLinearColor(Minimum=<rgba 0.2923 0.331 0.474 0.17>, Maximum=<rgba 0.1086 0.129 0.1667 0.22>)
    Lifetime=&UniformRangedFloat(Minimum=1, Maximum=1.5)
    Mesh Scale=&UniformRangedVector(Minimum=<0.65 0.65 1>, Maximum=<1 1 4>)
    Module.Write Color=true
    Module.Write Lifetime=true
    Module.Write Mass=false
    Module.Write Material Random=true
    Module.Write Position=false
    Module.Write Scale=true
    Module.Write SpriteRotation=false
    Module.Write SpriteSize=false
    Position=&SimulationPosition()
   )
   /Niagara/Modules/Spawn/Orientation/V2/InitialMeshOrientation(
    Rotation=&MakeVector(X=0, Y=&UniformAOrBFloat(A=0.75, B=-0.75), Z=0)
    Use Orientation Vector=true
    Use Rotation Vector=true
   )
   &DynamicMaterialParameters(
    Index 0 Param 1=&UniformRangedFloat(Minimum=0.1, Maximum=0.25)
    Index 0 Param 2=1
    Module.Param0WriteEnabled=true
    Module.Param1WriteEnabled=true
    Module.Param2WriteEnabled=false
    Module.Param3WriteEnabled=false
   )
   /Game/Effects/Niagara/Modules/Mesh/FN_MeshRenderer_v3/FN_MeshRenderer_v2(
    Color:LinearColor=&UniformRangedLinearColor(Minimum=<rgba 0.2923 0.331 0.474 0.3>, Maximum=<rgba 0.1086 0.129 0.1667 0.4>)
    Create COlor Attribute=true
    Create Dynamic Param Attribute=true
    Create Position Attribute=true
    DynamicMaterial:Vector4f=&MakeVector4(X=&UniformRangedFloat(Minimum=0.1, Maximum=0.25), Y=1, Z=0, W=0)
   )
   &CameraOffset(Camera Offset Amount=3)
  ]
  particleUpdateScript [
   &ParticleState()
   &SkeletalMeshLocation(
    Bone / Socket Index=4
    Bone Sampling Mode=::NewEnumerator4
    Inherit Velocity=false
    Interpolate Mesh Sampling=false
    MaskSpawnGroup=false
    Skeletal Mesh=$User.InstigatorMesh
    Use Sampled Mesh Attribute as Ribbon ID=false
   ) {enabled=false}
   /Niagara/Modules/Update/Orientation/V2/UpdateMeshOrientation(Rotation Rate:NiagaraFloat=&UniformRangedFloat(Minimum=1, Maximum=2), Rotation Vector:Vector3f=<1 0 0>)
   /SetVariables(Particles.Position=&Add_Vector(A=<25 1.5 -1>, B=&Multiply_VectorByFloat(Vector=<-28 -2.5 0>, Float=$Particles.NormalizedAge))) {isSetParameters=true}
   &ScaleMeshSize(Scale Factor=&VectorFromFloat(Value=&FloatFromCurve(FloatCurve=DataInterface() {propertyValues {Curve {keys [
          {arriveTangent=3.2175, interpMode=RCIM_Cubic, leaveTangent=2.5296, tangentMode=RCTM_User, time=0.25, value=0.75}
          {interpMode=RCIM_Cubic, leaveTangent=-0.394, time=0.9, value=1.25}
         ]}}})))
   &ScaleColor(
    Scale Alpha=&FloatFromCurve(FloatCurve=DataInterface() {propertyValues {Curve {keys [<at 0.65 1 RCIM_Cubic>,<at 0.85 0>]}}})
    ScaleA=true
    ScaleRGB=true
   )
  ]
  rendererClasses [&NiagaraMeshRendererProperties]
  renderers [{rendererClass=&NiagaraMeshRendererProperties, rendererIndex=0}]
 }
 Electric_Arcs_Hands {
  emitterUpdateScript [
   &EmitterState(
    Inactive Response=::NewEnumerator1
    Life Cycle Mode=::NewEnumerator1
    Loop Duration=&UniformRangedFloat(Minimum=0.14, Maximum=0.24)
    Recalculate Duration Each Loop=true
    UseLoopDelay=false
   )
   /SetVariables(Emitter.SpawnBurstRange:NiagaraInt32=&UniformRangedInt(Minimum=3, Maximum=4)) {isSetParameters=true}
   &SpawnBurst_Instantaneous(Spawn Count=$Emitter.SpawnBurstRange, Use Spawn Probability=false)
  ]
  particleSpawnScript [
   &InitializeParticle(
    Color=<rgba 3.099 4.4637 5 0.45>
    Lifetime=&UniformRangedFloat(Minimum=0.11, Maximum=0.14)
    Module.Write Color=true
    Module.Write Lifetime=true
    Module.Write Mass=true
    Module.Write Material Random=true
    Module.Write Position=true
    Module.Write Scale=false
    Module.Write SpriteRotation=false
    Module.Write SpriteSize=true
    Sprite Size=&UniformRangedVector2D(Minimum=<15 30>, Maximum=<20 45>)
   )
   &DynamicMaterialParameters(
    Index 0 Param 1=&UniformRangedFloat(Minimum=2.5, Maximum=3.5)
    Index 0 Param 3=&UniformRangedFloat(Minimum=0.2, Maximum=0.5)
    Module.Param0WriteEnabled=true
    Module.Param1WriteEnabled=false
    Module.Param2WriteEnabled=true
    Module.Param3WriteEnabled=true
   )
   &CameraOffset(Camera Offset Amount=&UniformRangedFloat(Minimum=-10, Maximum=10))
   /SetVariables(Particles.LightningUVScale:Vector2f=&MakeVector2D(X=&UniformAOrBFloat(A=-1), Y=&UniformAOrBFloat(A=-1))) {isSetParameters=true}
   /SetVariables(
    Particles.ShortLightningArcsColor:LinearColor=<rgba 0.9453 1.041 1.1 0.1>
    Particles.ShortLightningArcsSize:Vector2f=&Vector2DFromFloat(Value=&UniformRangedFloat(Minimum=12, Maximum=16))
   ) {isSetParameters=true}
  ]
  particleUpdateScript [
   &ParticleState()
   &SkeletalMeshLocation(
    Apply Sampled Orientation to Mesh Particles=true
    Apply To Particle Position=false
    Bone / Socket Index=3
    Bone Sampling Mode=::NewEnumerator4
    Inherit Velocity=false
    Interpolate Mesh Sampling=false
    MaskSpawnGroup=false
    Skeletal Mesh=$User.InstigatorMesh
   ) {module=SkeletalMeshLocation001}
   /SetVariables(
    Particles.LightningSpriteAlignment:Vector3f=&RotateVectorByQuaternion(Vector To Rotate=<-1 0 0>, Quaternion=$Output.SkeletalMeshLocation001.SampledRotation)
    Particles.Position:Vector3f=$Output.SkeletalMeshLocation001.SampledPosition
    Particles.ShortLightningArcsPosition:Vector3f=&Add_Vector(
     A=$Output.SkeletalMeshLocation001.SampledPosition
     B=&RotateVectorByQuaternion(Vector To Rotate=<-28 0 0>, Quaternion=$Output.SkeletalMeshLocation001.SampledRotation)
    )
   ) {isSetParameters=true}
   &FN_MeshRenderer_v2(
    Color:LinearColor=&MultiplyLinearColor(
     Linear Color=<rgba 0.6406 0.7354 0.75 0.1>
     Scale Factor=&MakeFloat4FromVectorAndFloat(
      "Vector (RGB)"=&MakeVectorFromLinearColorRGB(Color=&ColorFromCurve(ColorCurve=DataInterface() {propertyValues {
          AlphaCurve {keys []}
          BlueCurve {keys [<at 0 0.5 RCIM_Cubic>, {arriveTangent=-0.0312, interpMode=RCIM_Cubic, leaveTangent=-0.0312, time=0.2, value=2}, <at 1 0.4688 RCIM_Cubic>]}
          GreenCurve {keys [<at 0 0.3802 RCIM_Cubic>, {arriveTangent=0.1198, interpMode=RCIM_Cubic, leaveTangent=0.1198, time=0.2, value=1.8793}, <at 1 0.5 RCIM_Cubic>]}
          RedCurve {keys [
            <at 0 0.4422 RCIM_Cubic>
            {arriveTangent=0.0031, interpMode=RCIM_Cubic, leaveTangent=0.0031, time=0.2, value=1.5048}
            <at 1 0.4453 RCIM_Cubic>
           ]}
         }}))
      "Float (Alpha)"=&FloatFromCurve(FloatCurve=DataInterface() {propertyValues {Curve {keys [
           <at 0.1 1>
           {arriveTangent=-0.7039, interpMode=RCIM_Cubic, leaveTangent=-0.7039, time=0.3472, value=1}
           <at 0.5972 0.65 RCIM_Cubic>
           <at 0.8472 1>
           <at 1.0972 0.65 RCIM_Cubic>
           <at 1.3472 1 RCIM_Cubic>
           <at 1.5972 0.65>
           <at 1.8472 1 RCIM_Cubic>
          ]}}})
     )
    )
    Create COlor Attribute=true
    Create Position Attribute=true
    Create Scale Attribute=true
    Position Offset:Vector3f=&RotateVectorByQuaternion(Vector To Rotate=<-26 -1 3>, Quaternion=$Output.SkeletalMeshLocation001.SampledRotation)
    Scale:Vector3f=<0.95 0.7 0.7>
   ) {module=FN_MeshRenderer_v2001}
   &ScaleSpriteSize(Scale Factor=&Vector2DFromFloat(Value=&FloatFromCurve(FloatCurve=DataInterface() {propertyValues {Curve {keys [{arriveTangent=1.9395, interpMode=RCIM_Cubic, leaveTangent=1.9395, tangentMode=RCTM_User, time=0, value=0.3}, <at 1 1 RCIM_Cubic>]}}})))
   &FN_Light_v3(
    Color=&LerpLinearColors(StartColor=<rgba 0.8646 0.9483 1 1>, EndColor=$System.RandomColor, LerpFactor=0.5)
    Create Color Attribute=true
    Create Exponent Attribute=true
    Create Position Attribute=true
    Create Radius Attribute=true
    Enable Particle Light Enable=true
    Exponent=5
    Light Enabled=HlslExpression() {hlslExpression="Modulo (Particles.UniqueID, Emitter.SpawnBurstRange) == 1 ? 1.0: 0"}
    Position:Vector3f=&Add_Vector(
     A=$Output.SkeletalMeshLocation001.SampledPosition
     B=&RotateVectorByQuaternion(Vector To Rotate=<-34 2 10>, Quaternion=$Output.SkeletalMeshLocation001.SampledRotation)
    )
    Radius=70
    Scale Color:NiagaraFloat=&FloatFromCurve(FloatCurve=DataInterface() {propertyValues {Curve {keys [
         {arriveTangent=4.7823, interpMode=RCIM_Cubic, leaveTangent=4.7823, tangentMode=RCTM_User, time=0, value=0}
         <at 0.3 1 RCIM_Cubic>
         <at 1 0 RCIM_Cubic>
        ]}}})
   )
  ]
  rendererClasses [
   &NiagaraSpriteRendererProperties
   &NiagaraLightRendererProperties
   &NiagaraMeshRendererProperties
  ]
  renderers [
   {rendererClass=&NiagaraSpriteRendererProperties}
   {rendererClass=&NiagaraSpriteRendererProperties, rendererIndex=1}
   {rendererClass=&NiagaraLightRendererProperties, rendererIndex=2}
   {rendererClass=&NiagaraMeshRendererProperties, rendererIndex=3}
  ]
 }
 FN_BoneFlow_v2 {
  emitterUpdateScript [
   &EmitterState(
    Life Cycle Mode=::NewEnumerator1
    Loop Duration=&UniformRangedFloat(Minimum=0.175, Maximum=0.225)
    UseLoopDelay=false
   )
   /SetVariables(Emitter.RandomRangePartition:NiagaraInt32=&UniformRangedInt(Minimum=1, Maximum=20)) {isSetParameters=true}
   /Game/Effects/Niagara/Modules/Partition/FN_PartitionAttributes/FN_PartitionAttributes_v2(Particles Per Partition:NiagaraInt32=$Emitter.RandomRangePartition)
   &SpawnBurst_Instantaneous(Spawn Count=&UniformRangedInt(Minimum=1, Maximum=2), Use Spawn Probability=false)
  ]
  particleSpawnScript [
   /Game/Effects/Niagara/Modules/Partition/FN_ParticlePartition_v2/FN_ParticlePartition_v2()
   &InitializeParticle(
    Color=&UniformRangedLinearColor(Minimum=<rgba 1 0.7423 0.8026 1>, Maximum=<rgba 1.625 1.9872 2 1>)
    Lifetime=&UniformRangedFloat(Minimum=0.1, Maximum=0.15)
    Mesh Scale=&UniformRangedVector(Minimum=<0.4 0 0>, Maximum=<0.6 1 1>)
    Module.Write Color=true
    Module.Write Lifetime=true
    Module.Write Mass=true
    Module.Write Material Random=true
    Module.Write Position=false
    Module.Write Scale=true
    Module.Write SpriteRotation=true
    Module.Write SpriteSize=true
    Sprite Size=<1 1>
   )
   &DynamicMaterialParameters(
    Index 0 Param 1=&UniformRangedFloat(Minimum=1.5, Maximum=2)
    Index 0 Param 2=&UniformRangedFloat(Minimum=1, Maximum=2)
    Index 0 Param 3=&UniformRangedFloat(Minimum=0.5, Maximum=1)
    Index 0 Param 4=$Particles.MaterialRandom
    Index 1 Param 1=&UniformRangedFloat(Minimum=0, Maximum=1)
    Index 1 Param 2=&UniformRangedFloat(Minimum=0, Maximum=20)
    Module.Param0WriteEnabled=true
    Module.Param0WriteEnabled1=true
    Module.Param1WriteEnabled=true
    Module.Param1WriteEnabled1=true
    Module.Param2WriteEnabled=true
    Module.Param2WriteEnabled1=false
    Module.Param3WriteEnabled=true
    Module.Param3WriteEnabled1=false
    Write Parameter Index 1=true
   )
   /Game/Effects/Niagara/Modules/BoneFlow/FN_Initialize_BoneFlow_Attributes_v2(Angle=1, Radius=1)
   &InitialMeshOrientation(
    Orientation Vector=<-1 0 0>
    Rotation=&UniformRangedVector(Minimum=<-1 0 0>, Maximum=<1 0 0>)
    Rotation Coordinate Space:ENiagaraMeshTransforms=::NewEnumerator0
    Use Orientation Vector=true
    Use Rotation Vector=true
   )
  ]
  particleUpdateScript [
   &ParticleState()
   /Game/Effects/Niagara/Modules/BoneFlow/FN_BoneFlow_v2/FN_BoneFlow_v2(
    Radius:NiagaraFloat=$Particles.Initial.FN_Initialize_BoneFlow_Attributes_v2.Radius
    SkeletalMesh:NiagaraDataInterfaceSkeletalMesh=$User.InstigatorMesh
    T:NiagaraFloat=&Add_Float(A=$Particles.FN_Initialize_BoneFlow_Attributes_v2.T, B=&Multiply_Float(A=$Particles.NormalizedAge, B=0.025))
   )
   &SkeletalMeshLocation(
    Apply Sampled Orientation to Mesh Particles=true
    Apply To Particle Position=false
    Bone / Socket Index=3
    Bone Sampling Mode=::NewEnumerator4
    Inherit Velocity=false
    MaskSpawnGroup=false
    Skeletal Mesh=$User.InstigatorMesh
   )
   &ScaleColor(
    Scale Alpha=&FloatFromCurve(
     FloatCurve=DataInterface() {propertyValues {Curve {keys [<at 0.8 1 RCIM_Cubic>, {interpMode=RCIM_Cubic, leaveTangent=-12.7776, time=1, value=0}]}}}
     Scale Curve=&self:Flicker(FlickerAmount:NiagaraFloat=1, FlickerInterval:NiagaraFloat=0.1)
    )
    Scale RGB=&MakeVectorFromLinearColorRGB(Color=&ColorFromCurve(ColorCurve=DataInterface() {propertyValues {
        AlphaCurve {keys []}
        BlueCurve {keys [<at 0 1 RCIM_Cubic>,<at 1 3 RCIM_Cubic>]}
        GreenCurve {keys [<at 0 0.9955 RCIM_Cubic>,<at 1 2.8618 RCIM_Cubic>]}
        RedCurve {keys [<at 0 0.9844 RCIM_Cubic>,<at 1 2.6719 RCIM_Cubic>]}
       }}))
    ScaleA=true
    ScaleRGB=true
   )
   &ScaleMeshSize(Scale Factor=&MakeVector(
     X=&FloatFromCurve(FloatCurve=DataInterface() {propertyValues {Curve {keys [<at 0 0.5 RCIM_Cubic>, {arriveTangent=1, leaveTangent=1, time=0.25, value=1}, <at 1 1>]}}})
     Y=1
     Z=&FloatFromCurve(FloatCurve=DataInterface() {propertyValues {Curve {keys [<at 0.8 1 RCIM_Cubic>, {arriveTangent=-11.4579, interpMode=RCIM_Cubic, leaveTangent=-11.4579, tangentMode=RCTM_User, time=1, value=0}]}}})
    ))
   /SetVariables(Particles.MeshOrientation=&MultiplyQuaternion(Quaternion A=$Output.SkeletalMeshLocation.SampledRotation, Quaternion B=$Particles.Initial.MeshOrientation)) {isSetParameters=true}
  ]
  rendererClasses [&NiagaraMeshRendererProperties]
  renderers [{rendererClass=&NiagaraMeshRendererProperties, rendererIndex=0}]
 }
 Flares {
  emitterUpdateScript [
   &EmitterState(
    Life Cycle Mode=::NewEnumerator1
    Loop Duration=&UniformRangedFloat(Minimum=0.15, Maximum=0.22)
    UseLoopDelay=false
   )
   &SpawnBurst_Instantaneous(Spawn Count=&UniformRangedInt(Minimum=1, Maximum=2), Use Spawn Probability=false)
   &SpawnBurst_Instantaneous(Spawn Count=&UniformRangedInt(Minimum=1, Maximum=2), Spawn Time=0.12, Use Spawn Probability=false) {module=SpawnBurst_Instantaneous001}
  ]
  particleSpawnScript [
   &InitializeParticle(
    Color=<rgba 1.35 1.4336 1.5 1>
    Lifetime=&UniformRangedFloat(Minimum=0.14, Maximum=0.2)
    Module.Write Color=true
    Module.Write Lifetime=true
    Module.Write Mass=true
    Module.Write Material Random=true
    Module.Write Position=false
    Module.Write Scale=false
    Module.Write SpriteRotation=true
    Module.Write SpriteSize=true
    Sprite Size=&Vector2DFromFloat(Value=&UniformRangedFloat(Minimum=10, Maximum=20))
   )
   &DynamicMaterialParameters(
    Index 0 Param 1=&UniformRangedFloat(Minimum=0, Maximum=0.3)
    Index 0 Param 2=&UniformRangedFloat(Minimum=1.2, Maximum=4)
    Index 0 Param 4=1
    Module.Param0WriteEnabled=true
    Module.Param1WriteEnabled=true
    Module.Param2WriteEnabled=false
    Module.Param3WriteEnabled=false
   )
   /Niagara/Modules/Spawn/Location/SphereLocation(
    Coordinate Space:ENiagaraCoordinateSpace=::Simulation
    MaskSpawnGroup=false
    Module.Override Randomness=false
    Non Uniform Scale=<0.95 0.2 0.45>
    Sphere Origin:NiagaraPosition=&Add_Vector(A=Linked() {linkedVariable {name=Particles.Position, type=NiagaraPosition}}, B=<-6 0 0>)
    Sphere Radius=8
    Surface Only=false
   )
   &CameraOffset(Camera Offset Amount=-5)
   /SetVariables(Particles.Color=&MultiplyLinearColor(Linear Color=$Particles.Color, Scale Factor=&MakeVector4_FromLinearColor(LinearColor=$System.RandomColor))) {isSetParameters=true, module=SetVariables_#2}
  ]
  particleUpdateScript [
   &ParticleState()
   &DynamicMaterialParameters(
    Index 0 Param 4=&FloatFromCurve(FloatCurve=DataInterface() {propertyValues {Curve {keys [<at 0 1>,<at 0.2 0>,<at 0.4625 0.6907>,<at 1 1>]}}})
    Module.Param0WriteEnabled=false
    Module.Param1WriteEnabled=false
    Module.Param2WriteEnabled=false
    Module.Param3WriteEnabled=true
   ) {module=DynamicMaterialParameters001}
   &ScaleColor(
    Scale Alpha=&FloatFromCurve(FloatCurve=DataInterface() {propertyValues {Curve {keys [<at 0 0 RCIM_Cubic>, <at 0.25 0.4>, {arriveTangent=-0.5333, interpMode=RCIM_Cubic, leaveTangent=-0.5333, time=0.75, value=0.4}, <at 1 0>]}}})
    Scale RGB=&MakeVectorFromLinearColorRGB(Color=&ColorFromCurve(ColorCurve=DataInterface() {propertyValues {
        AlphaCurve {keys []}
        BlueCurve {keys [<at 0 1 RCIM_Cubic>,<at 0.2 1.75 RCIM_Cubic>,<at 0.693 1>]}
        GreenCurve {keys [<at 0 0.9955 RCIM_Cubic>, {arriveTangent=0.0065, interpMode=RCIM_Cubic, leaveTangent=0.0065, time=0.2, value=1.6694}, <at 0.693 1>]}
        RedCurve {keys [<at 0 0.9844 RCIM_Cubic>, {arriveTangent=0.0225, interpMode=RCIM_Cubic, leaveTangent=0.0225, time=0.2, value=1.5586}, <at 0.693 1>]}
       }}))
    ScaleA=true
    ScaleRGB=true
   )
  ]
  rendererClasses [&NiagaraSpriteRendererProperties]
  renderers [{rendererClass=&NiagaraSpriteRendererProperties}]
 }
 Lightning_Blades {
  emitterUpdateScript [
   &EmitterState(
    Life Cycle Mode=::NewEnumerator1
    Loop Duration=&UniformRangedFloat(Minimum=0.15, Maximum=0.2)
    Recalculate Duration Each Loop=true
    UseLoopDelay=false
   )
   &SpawnBurst_Instantaneous(Spawn Count=&UniformRangedInt(Minimum=1, Maximum=2), Use Spawn Probability=false)
   &SpawnBurst_Instantaneous(Spawn Count=1, Spawn Time=0.12, Use Spawn Probability=false) {module=SpawnBurst_Instantaneous001}
  ]
  particleSpawnScript [
   &InitializeParticle(
    Color=&UniformRangedLinearColor(Minimum=<rgba 0.5 0.3712 0.4013 1>, Maximum=<rgba 0.8125 0.9936 1 1>)
    Lifetime=&UniformRangedFloat(Minimum=0.075, Maximum=0.15)
    Mesh Scale=&UniformRangedVector(Minimum=<0.5 0.5 0.5>, Maximum=<1 1 1>)
    Module.Write Color=true
    Module.Write Lifetime=true
    Module.Write Mass=false
    Module.Write Material Random=true
    Module.Write Position=false
    Module.Write Scale=true
    Module.Write SpriteRotation=false
    Module.Write SpriteSize=false
    Position=&SimulationPosition()
   )
   &InitialMeshOrientation(
    Rotation=&UniformRangedVector(Minimum=<-1 0 0>, Maximum=<1 0 0>)
    Use Orientation Vector=true
    Use Rotation Vector=true
   )
   &DynamicMaterialParameters(
    Index 0 Param 1=&UniformRangedFloat(Minimum=2, Maximum=3)
    Index 0 Param 2=&UniformRangedFloat(Minimum=0.5, Maximum=1)
    Index 0 Param 3=&UniformRangedFloat(Minimum=2, Maximum=4)
    Index 0 Param 4=&UniformRangedFloat(Minimum=0, Maximum=50)
    Index 1 Param 1=&UniformRangedFloat(Minimum=1, Maximum=2)
    Index 1 Param 2=&UniformRangedFloat(Minimum=0, Maximum=40)
    Module.Param0WriteEnabled=true
    Module.Param0WriteEnabled1=true
    Module.Param1WriteEnabled=true
    Module.Param1WriteEnabled1=true
    Module.Param2WriteEnabled=true
    Module.Param2WriteEnabled1=false
    Module.Param3WriteEnabled=true
    Module.Param3WriteEnabled1=false
    Write Parameter Index 1=true
   )
   &SphereLocation(MaskSpawnGroup=false, Module.Override Randomness=false, Sphere Radius=3, Surface Only=false) {enabled=false}
  ]
  particleUpdateScript [
   &ParticleState()
   &SkeletalMeshLocation(
    Apply Sampled Orientation to Mesh Particles=true
    Apply To Particle Position=false
    Bone / Socket Index=4
    Bone Sampling Mode=::NewEnumerator4
    Inherit Velocity=false
    Interpolate Mesh Sampling=false
    MaskSpawnGroup=false
    Skeletal Mesh=$User.InstigatorMesh
    Use Sampled Mesh Attribute as Ribbon ID=false
   )
   &ScaleColor(
    Scale Alpha=&FloatFromCurve(FloatCurve=DataInterface() {propertyValues {Curve {keys [<at 0.15 1>,<at 0.9 1>,<at 1 0>]}}})
    Scale RGB=&MakeVectorFromLinearColorRGB(Color=&ColorFromCurve(ColorCurve=DataInterface() {propertyValues {
        AlphaCurve {keys []}
        BlueCurve {keys [<at 0 5.5>,<at 0.162 6.6988>,<at 0.2527 2.0661>,<at 1 1>]}
        GreenCurve {keys [<at 0 5.5>,<at 0.162 6.6988>,<at 0.2527 2.0661>,<at 1 1>]}
        RedCurve {keys [<at 0 5.5>,<at 0.162 6.6988>,<at 0.2527 2.0661>,<at 1 1>]}
       }}))
    ScaleA=true
    ScaleRGB=true
   )
   &ScaleMeshSize(Scale Factor=&MakeVector(
     X=&FloatFromCurve(FloatCurve=DataInterface() {propertyValues {Curve {keys [<at 0 0.7 RCIM_Cubic>, {arriveTangent=0.15, leaveTangent=0.15, time=0.3, value=1}]}}})
     Y=1
     Z=&FloatFromCurve(FloatCurve=DataInterface() {propertyValues {Curve {keys [<at 0.8 1 RCIM_Cubic>, {arriveTangent=-11.4579, interpMode=RCIM_Cubic, leaveTangent=-11.4579, tangentMode=RCTM_User, time=1, value=0}]}}})
    ))
   /SetVariables(Particles.MeshOrientation=&MultiplyQuaternion(Quaternion A=$Output.SkeletalMeshLocation.SampledRotation, Quaternion B=$Particles.Initial.MeshOrientation)) {isSetParameters=true}
  ]
  rendererClasses [&NiagaraMeshRendererProperties]
  renderers [{rendererClass=&NiagaraMeshRendererProperties, rendererIndex=0}]
 }
 Lightning_Main_Blade {
  emitterUpdateScript [
   &EmitterState(
    Life Cycle Mode=::NewEnumerator1
    Loop Duration=&UniformRangedFloat(Minimum=0.08, Maximum=0.11)
    Recalculate Duration Each Loop=true
    UseLoopDelay=false
   )
   &SpawnBurst_Instantaneous(Spawn Count=1, Use Spawn Probability=false)
  ]
  particleSpawnScript [
   &InitializeParticle(
    Color=&UniformRangedLinearColor(Minimum=<rgba 0.4 0.2969 0.321 1.25>, Maximum=<rgba 0.5687 0.6955 0.7 1.25>, Color Channel Mode=::NewEnumerator1)
    Lifetime=&UniformRangedFloat(Minimum=0.1, Maximum=0.16)
    Mesh Scale=&UniformRangedVector(Minimum=<-0.45 1 1>, Maximum=<-0.7 1 1>)
    Module.Write Color=true
    Module.Write Lifetime=true
    Module.Write Mass=false
    Module.Write Material Random=true
    Module.Write Position=false
    Module.Write Scale=true
    Module.Write SpriteRotation=false
    Module.Write SpriteSize=false
    Position=&SimulationPosition()
   )
   &DynamicMaterialParameters(
    Index 0 Param 1=&UniformRangedFloat(Minimum=7, Maximum=13)
    Index 0 Param 2=&UniformRangedFloat(Minimum=0.2, Maximum=1.1)
    Index 0 Param 3=&UniformRangedFloat(Minimum=0, Maximum=1)
    Index 0 Param 4=&UniformRangedFloat(Minimum=0, Maximum=50)
    Index 1 Param 1=&UniformRangedFloat(Minimum=0.5, Maximum=1)
    Index 1 Param 2=&UniformRangedFloat(Minimum=0, Maximum=20)
    Module.Param0WriteEnabled=true
    Module.Param0WriteEnabled1=true
    Module.Param1WriteEnabled=true
    Module.Param1WriteEnabled1=true
    Module.Param2WriteEnabled=true
    Module.Param2WriteEnabled1=false
    Module.Param3WriteEnabled=true
    Module.Param3WriteEnabled1=false
    Write Parameter Index 1=true
   )
  ]
  particleUpdateScript [
   &ParticleState()
   &DynamicMaterialParameters(Module.Param0WriteEnabled=false, Module.Param1WriteEnabled=false, Module.Param2WriteEnabled=false, Module.Param3WriteEnabled=false) {enabled=false, module=DynamicMaterialParameters001}
   &SkeletalMeshLocation(
    Apply Sampled Orientation to Mesh Particles=true
    Apply To Particle Position=false
    Bone / Socket Index=4
    Bone Sampling Mode=::NewEnumerator4
    Inherit Velocity=false
    MaskSpawnGroup=false
    Skeletal Mesh=$User.InstigatorMesh
    Use Sampled Mesh Attribute as Ribbon ID=false
   )
   &ScaleColor(
    Scale Alpha=&FloatFromCurve(FloatCurve=DataInterface() {propertyValues {Curve {keys [<at 0.15 1>,<at 0.9 1>,<at 1 0>]}}})
    Scale RGB=&MakeVectorFromLinearColorRGB(Color=&ColorFromCurve(ColorCurve=DataInterface() {propertyValues {
        AlphaCurve {keys []}
        BlueCurve {keys [<at 0 2.5>,<at 0.162 3.5>,<at 0.253 1.25>,<at 1 1>]}
        GreenCurve {keys [<at 0 2.5>,<at 0.162 3.5>,<at 0.253 1.25>,<at 1 1>]}
        RedCurve {keys [<at 0 2.5>,<at 0.162 3.5>,<at 0.253 1.25>,<at 1 1>]}
       }}))
    ScaleA=true
    ScaleRGB=true
   )
   &ScaleMeshSize(Scale Factor=&MakeVector(
     X=&FloatFromCurve(FloatCurve=DataInterface() {propertyValues {Curve {keys [<at 0 0.7 RCIM_Cubic>, {arriveTangent=0.15, leaveTangent=0.15, time=0.3, value=1}]}}})
     Y=1
     Z=1
    ))
   /SetVariables(Particles.Position=&Add_Vector(
     A=Linked() {linkedVariable {name=Engine.Owner.Position, type=NiagaraPosition}}
     B=&RotateVectorByQuaternion(Vector To Rotate=<-2 0 0>, Quaternion=$Output.SkeletalMeshLocation.SampledRotation)
    )) {isSetParameters=true}
  ]
  rendererClasses [&NiagaraMeshRendererProperties]
  renderers [{rendererClass=&NiagaraMeshRendererProperties, rendererIndex=0}]
 }
 Particle_Decal {
  bEnabled=false
  emitterUpdateScript [
   &EmitterState(
    Life Cycle Mode=::NewEnumerator1
    Loop Delay=0.5
    Loop Duration=&UniformRangedFloat(Minimum=0.75, Maximum=1)
    Recalculate Duration Each Loop=true
    UseLoopDelay=false
   )
   &SpawnRate(SpawnRate=2, Use Spawn Probability=false)
  ]
  particleSpawnScript [
   &InitializeParticle(
    Color=<rgba 0.761 0.761 0.761 0.65>
    Mesh Scale=&UniformRangedVector(Minimum=<0.65 0.65 0.65>, Maximum=<1 1 1>)
    Module.Write Color=true
    Module.Write Lifetime=true
    Module.Write Mass=false
    Module.Write Material Random=true
    Module.Write Position=false
    Module.Write Scale=true
    Module.Write SpriteRotation=true
    Module.Write SpriteSize=true
    Position=&SimulationPosition()
    Sprite Rotation=&NormalizedAngleToDegrees(NormalizedAngle=&UniformRangedFloat(Minimum=-1, Maximum=1))
    Sprite Size=&Vector2DFromFloat(Value=&UniformRangedFloat(Minimum=40, Maximum=40))
   )
   &DynamicMaterialParameters(
    Index 0 Param 1=0.75
    Index 0 Param 2=50
    Index 0 Param 3=&UniformRangedFloat(Minimum=0, Maximum=1)
    Index 0 Param 4=&UniformRangedFloat(Minimum=0, Maximum=1)
    Module.Param0WriteEnabled=true
    Module.Param1WriteEnabled=true
    Module.Param2WriteEnabled=true
    Module.Param3WriteEnabled=true
   )
   &CameraOffset(Camera Offset Amount=10)
  ]
  particleUpdateScript [
   &ParticleState()
   &SkeletalMeshLocation(
    Bone / Socket Index=4
    Bone Sampling Mode=::NewEnumerator4
    Inherit Velocity=false
    MaskSpawnGroup=false
    Skeletal Mesh=$User.InstigatorMesh
    Use Sampled Mesh Attribute as Ribbon ID=false
   )
   /SetVariables(Particles.Position=&Add_Vector(A=$Output.SkeletalMeshLocation.SampledPosition, B=<-4 0 0>)) {isSetParameters=true, module=SetVariables_#3}
   &ScaleColor(
    Scale Alpha=&FloatFromCurve(FloatCurve=DataInterface() {propertyValues {Curve {keys [<at 0 0>,<at 0.1 1>,<at 0.75 1>,<at 1 0>]}}})
    ScaleA=true
    ScaleRGB=true
   )
  ]
  rendererClasses [&NiagaraSpriteRendererProperties]
  renderers [{rendererClass=&NiagaraSpriteRendererProperties}]
 }
 Particle_Decal001 {
  emitterUpdateScript [
   &EmitterState(
    Life Cycle Mode=::NewEnumerator1
    Loop Delay=0.5
    Loop Duration=&UniformRangedFloat(Minimum=0.75, Maximum=1)
    Recalculate Duration Each Loop=true
    UseLoopDelay=false
   )
   &SpawnRate(SpawnRate=7, Use Spawn Probability=false)
  ]
  particleSpawnScript [
   &InitializeParticle(
    Color=<rgba 0.8282 0.7363 1 2>
    Lifetime=&UniformRangedFloat(Minimum=0.4, Maximum=0.7)
    Module.Write Color=true
    Module.Write Lifetime=true
    Module.Write Mass=false
    Module.Write Material Random=true
    Module.Write Position=true
    Module.Write Scale=false
    Module.Write SpriteRotation=true
    Module.Write SpriteSize=true
    Position=&UniformRangedVector(Minimum=<-6 -1 -1>, Maximum=<14 1 1>)
    Sprite Rotation=&NormalizedAngleToDegrees(NormalizedAngle=&UniformRangedFloat(Minimum=0.4, Maximum=0.5))
    Sprite Size=&Vector2DFromFloat(Value=&UniformRangedFloat(Minimum=15, Maximum=22))
   )
   &DynamicMaterialParameters(
    Index 0 Param 1=&UniformRangedFloat(Minimum=1, Maximum=2.5)
    Index 0 Param 4=$Particles.MaterialRandom
    Module.Param0WriteEnabled=true
    Module.Param1WriteEnabled=false
    Module.Param2WriteEnabled=false
    Module.Param3WriteEnabled=true
   )
   &CameraOffset(Camera Offset Amount=16)
   /SetVariables(Particles.Color=&MultiplyLinearColor(Linear Color=$Particles.Color, Scale Factor=&MakeVector4_FromLinearColor(LinearColor=$System.RandomColor))) {isSetParameters=true, module=SetVariables_#2}
  ]
  particleUpdateScript [
   &ParticleState()
   &SkeletalMeshLocation(
    Bone / Socket Index=4
    Bone Sampling Mode=::NewEnumerator4
    Inherit Velocity=false
    Interpolate Mesh Sampling=false
    MaskSpawnGroup=false
    Skeletal Mesh=$User.InstigatorMesh
    Use Sampled Mesh Attribute as Ribbon ID=false
   ) {enabled=false}
   /SetVariables(Particles.Position=&Add_Vector(A=$Output.SkeletalMeshLocation.SampledPosition, B=<-8 1 1.25>)) {enabled=false, isSetParameters=true, module=SetVariables_#3}
   &ScaleColor(
    Scale Alpha=&FloatFromCurve(
     FloatCurve=DataInterface() {propertyValues {Curve {keys [
         {arriveTangent=4.6264, interpMode=RCIM_Cubic, leaveTangent=4.6264, tangentMode=RCTM_User, time=0, value=0}
         {interpMode=RCIM_Cubic, tangentMode=RCTM_User, time=0.4, value=1}
         {arriveTangent=-2.8472, interpMode=RCIM_Cubic, leaveTangent=-2.8472, tangentMode=RCTM_User, time=1, value=0}
        ]}}}
     Scale Curve=&self:Flicker(FlickerAmount:NiagaraFloat=0.98, FlickerInterval:NiagaraFloat=0.1)
    )
    ScaleA=true
    ScaleRGB=false
   )
   &DynamicMaterialParameters(
    Index 0 Param 3=$Particles.NormalizedAge
    Module.Param0WriteEnabled=false
    Module.Param1WriteEnabled=false
    Module.Param2WriteEnabled=true
    Module.Param3WriteEnabled=false
   ) {module=DynamicMaterialParameters001}
  ]
  rendererClasses [&NiagaraSpriteRendererProperties]
  renderers [{rendererClass=&NiagaraSpriteRendererProperties}]
 }
 Particle_Decal002 {
  emitterUpdateScript [
   &EmitterState(
    Life Cycle Mode=::NewEnumerator1
    Loop Delay=0.5
    Loop Duration=&UniformRangedFloat(Minimum=0.75, Maximum=1)
    Recalculate Duration Each Loop=true
    UseLoopDelay=false
   )
   &SpawnRate(SpawnRate=16, Use Spawn Probability=false)
  ]
  particleSpawnScript [
   &InitializeParticle(
    Color=<rgba 6.9583 7.7462 8 0.25>
    Lifetime=&UniformRangedFloat(Minimum=0.3, Maximum=0.5)
    Module.Write Color=true
    Module.Write Lifetime=true
    Module.Write Mass=false
    Module.Write Material Random=true
    Module.Write Position=true
    Module.Write Scale=false
    Module.Write SpriteRotation=true
    Module.Write SpriteSize=true
    Position=&UniformRangedVector(Minimum=<10 0 0>, Maximum=<25 0 0>)
    Sprite Rotation=&NormalizedAngleToDegrees(NormalizedAngle=&UniformRangedFloat(Minimum=-1, Maximum=1))
    Sprite Size=&Vector2DFromFloat(Value=&UniformRangedFloat(Minimum=15, Maximum=22))
   )
   &DynamicMaterialParameters(
    Index 0 Param 2=&UniformRangedFloat(Minimum=0, Maximum=10)
    Index 0 Param 3=1
    Index 0 Param 4=1
    Module.Param0WriteEnabled=false
    Module.Param1WriteEnabled=true
    Module.Param2WriteEnabled=true
    Module.Param3WriteEnabled=true
   )
   &CameraOffset(Camera Offset Amount=7)
   &AddVelocity(Velocity=&UniformRangedVector(Minimum=<-35 0 0>, Maximum=<-25 0 0>))
  ]
  particleUpdateScript [
   &ParticleState()
   &SkeletalMeshLocation(
    Bone / Socket Index=4
    Bone Sampling Mode=::NewEnumerator4
    Inherit Velocity=false
    Interpolate Mesh Sampling=false
    MaskSpawnGroup=false
    Skeletal Mesh=$User.InstigatorMesh
    Use Sampled Mesh Attribute as Ribbon ID=false
   ) {enabled=false}
   &DynamicMaterialParameters(
    Index 0 Param 1=&FloatFromCurve(FloatCurve=DataInterface() {propertyValues {Curve {keys [<at 0 0.01>,<at 1 1.5>]}}})
    Module.Param0WriteEnabled=true
    Module.Param1WriteEnabled=false
    Module.Param2WriteEnabled=false
    Module.Param3WriteEnabled=false
   ) {module=DynamicMaterialParameters001}
   /SetVariables(Particles.Position=&Add_Vector(
     A=$Output.SkeletalMeshLocation.SampledPosition
     B=&Multiply_VectorByFloat(Vector=<-9 -0.5 1.25>, Float=$Particles.NormalizedAge)
    )) {enabled=false, isSetParameters=true, module=SetVariables_#3}
   &ScaleColor(
    Scale Alpha=&FloatFromCurve(FloatCurve=DataInterface() {propertyValues {Curve {keys [<at 0 1>,<at 0.2 0>,<at 1 0.5>]}}})
    Scale RGB=&VectorFromFloat(Value=&FloatFromCurve(FloatCurve=DataInterface() {propertyValues {Curve {keys [
          <at 0 1 RCIM_Cubic>
          {arriveTangent=0.875, interpMode=RCIM_Cubic, leaveTangent=0.875, time=0.1277, value=1.2582}
          {arriveTangent=-0.2961, interpMode=RCIM_Cubic, leaveTangent=-0.2961, time=0.3094, value=1.2707}
          <at 1 1 RCIM_Cubic>
         ]}}}))
    ScaleA=true
    ScaleRGB=true
   )
   &ScaleSpriteSize(Scale Factor=&Vector2DFromFloat(Value=&FloatFromCurve(FloatCurve=DataInterface() {propertyValues {Curve {keys [<at 0 0.15>,<at 0.4522 0.5145>,<at 0.9 1>]}}})))
   &SolveForcesAndVelocity(Clamp Velocity=false, Limit Acceleration=false)
  ]
  rendererClasses [&NiagaraSpriteRendererProperties]
  renderers [{rendererClass=&NiagaraSpriteRendererProperties}]
 }
}
summary {
 systemName=NS_PIckaxe_HighTowerSquash_1H_Idle
 emitters [
  {
   emitterName=Electric_Arcs_Hands
   rendererClasses [
    &NiagaraSpriteRendererProperties
    &NiagaraLightRendererProperties
    &NiagaraMeshRendererProperties
   ]
  }
  {emitterName=Debris_Vortex, rendererClasses [&NiagaraMeshRendererProperties]}
  {emitterName=Lightning_Blades, rendererClasses [&NiagaraMeshRendererProperties]}
  {bEnabled=false, emitterName=Particle_Decal, rendererClasses [&NiagaraSpriteRendererProperties]}
  {emitterName=FN_BoneFlow_v2, rendererClasses [&NiagaraMeshRendererProperties]}
  {emitterName=Lightning_Main_Blade, rendererClasses [&NiagaraMeshRendererProperties]}
  {emitterName=Flares, rendererClasses [&NiagaraSpriteRendererProperties]}
  {emitterName=Particle_Decal001, rendererClasses [&NiagaraSpriteRendererProperties]}
  {emitterName=Particle_Decal002, rendererClasses [&NiagaraSpriteRendererProperties]}
 ]
 userVariables [User.InstigatorMesh {
   type=NiagaraDataInterfaceSkeletalMesh
   defaultValue {
    struct=NiagaraExt_VariableValue_DataInterface
    value {dataInterface=&self:NiagaraDataInterfaceSkeletalMesh_0}
   }
  }]
}
