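# Niagara system for the swing trail of the one-handed HighTower Squash pickaxe, judging by its name, built from a single ribbon emitter.
# Short-lived, faint ribbon particles spawn continuously, offset backward along the owner's X axis, with ribbon facing locked to its Y axis.
# A color curve indexed by emitter loop age, not particle age, scales RGB from overbright down to normal across each loop.
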NiagaraSystem /BRCosmetics/Weapons/FORT_Melee/Pickaxe_HighTower_Squash_1H/FX/NS_Pickaxe_HighTower_Squash_1H_Animtrail
dependencies {usedDataInterfaces [&NiagaraDataInterfaceColorCurve], usedDynamicInputs [], usedModules [&SystemState]}
emitters {FN_Ribbon {
  emitterUpdateScript [
   &EmitterState()
   &SpawnRate(SpawnRate=75, Use Spawn Probability=false)
  ]
  particleSpawnScript [
   &InitializeRibbon(
    Color=<rgba 0.8542 1 1 0.08>
    Lifetime=0.1
    Module.Write Color=true
    Module.Write Lifetime=true
    Module.Write Mass=true
    Module.Write Material Random=true
    Module.Write Position=true
    Module.Write RibbonTwist=false
    Module.Write RibbonWidth=true
    Position:Vector3f=&Add_Vector(
     A=/Niagara/DynamicInputs/Helpers/SimulationPosition()
     B=&Multiply_VectorByFloat(Vector=$Engine.Owner.SystemXAxis, Float=-30)
    )
    Ribbon Width=50
   )
   /SetVariables(Particles.RibbonFacing=$Engine.Owner.SystemYAxis) {isSetParameters=true}
  ]
  particleUpdateScript [
   &ParticleState()
   &ScaleColor(
    Scale RGB=&MakeVectorFromLinearColorRGB(Color=&ColorFromCurve(
      ColorCurve=DataInterface() {propertyValues {
        AlphaCurve {keys [<at 0 1 RCIM_Cubic>,<at 1 0 RCIM_Cubic>]}
        BlueCurve {keys [<at 0 10 RCIM_Cubic>,<at 1 1 RCIM_Cubic>]}
        GreenCurve {keys [<at 0 10 RCIM_Cubic>,<at 1 1 RCIM_Cubic>]}
        RedCurve {keys [<at 0 10 RCIM_Cubic>,<at 1 1 RCIM_Cubic>]}
       }}
      CurveIndex=$Emitter.NormalizedLoopAge
     ))
    ScaleA=true
    ScaleRGB=true
   )
  ]
  rendererClasses [&NiagaraRibbonRendererProperties]
  renderers [{rendererClass=&NiagaraRibbonRendererProperties}]
 }}
summary {
 systemName=NS_Pickaxe_HighTower_Squash_1H_Animtrail
 emitters [{emitterName=FN_Ribbon, rendererClasses [&NiagaraRibbonRendererProperties]}]
 userVariables []
}
